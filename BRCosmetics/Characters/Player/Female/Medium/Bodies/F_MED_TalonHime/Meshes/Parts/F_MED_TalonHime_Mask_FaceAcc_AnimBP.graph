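# Animation Blueprint for the TalonHime female character's masked face-accessory part, adding secondary motion to ponytail, bangs, side and back hair.
# It copies the parent mesh pose, then chains constraint and Look At aiming with spring Anim Dynamics using sphere and plane limits.
# Head-versus-spine pitch drives corrective offsets, zeroed during melee animations, while jumping tilts hair base bones.
# A machine-gun correction is computed for GoingCommando weapons but unused in the AnimGraph.
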AnimBlueprint /BRCosmetics/Characters/Player/Female/Medium/Bodies/F_MED_TalonHime/Meshes/Parts/F_MED_TalonHime_Mask_FaceAcc_AnimBP
target_skeleton=/Game/Characters/Player/Female/Medium/Bodies/F_MED_TalonHime/Meshes/Parts/F_MED_TalonHime_Mask_FaceAcc_Skeleton
graphs [
 AnimGraph {nodes [
   n1 = Anim Dynamics - Bone: dyn_ponytail(ComponentPose=Local To Component(LocalPose=Copy Pose From Mesh() {bUseAttachedParent=true}), Alpha=1, ExternalForce="0,0,0") {
    angularDampingOverride=1
    angularSpringConstant=120
    bAngularSpring=true
    bOverrideAngularDamping=true
    bUseSphericalLimits=true
    gravityScale=0
    alphaBoolBlend {blendOption=Linear}
    boundBone {boneName=dyn_ponytail}
    componentAppliedLinearAccClamp=<300 300 200>
    componentLinearAccScale=<0.15 0.15 0.1>
    componentLinearVelScale=<0.3 0.3 0.15>
    physicsBodyDefinitions [{
      collisionType=CustomSphere
      sphereCollisionRadius=3
      boundBone {boneName=dyn_ponytail}
      boxExtents=<5 9 10>
      constraintSetup {angularLimitsMax=<0 30 15>, angularLimitsMin=<0 -30 -20>, angularTarget=<0 0 0>}
      localJointOffset=<0.7 0 0>
     }]
    relativeSpaceBone {boneName=spine_05}
    sphericalLimits [{limitRadius=7, limitType=Inner, drivingBone {boneName=head}, sphereLocalOffset=<15.5674 -8.493 0.2462>}]
   }
   n2 = Anim Dynamics - Bone: dyn_hairBang_maskOn_l(ComponentPose=@n1, Alpha=1, ExternalForce="0,0,0") {
    angularDampingOverride=0.9999
    angularSpringConstant=1500
    bAngularSpring=true
    bGravityOverrideInSimSpace=true
    bOverrideAngularDamping=true
    bUseSphericalLimits=true
    gravityScale=0.35
    simulationSpace=RootRelative
    alphaBoolBlend {blendOption=Linear}
    boundBone {boneName=dyn_hairBang_maskOn_l}
    componentAppliedLinearAccClamp=<300 300 300>
    componentLinearAccScale=<0.15 0.15 0.15>
    componentLinearVelScale=<0.3 0.3 0.3>
    physicsBodyDefinitions [{
      sphereCollisionRadius=0
      boundBone {boneName=dyn_hairBang_maskOn_l}
      boxExtents=<6 5 6>
      constraintSetup {angularLimitsMax=<0 0 40>, angularLimitsMin=<0 -15 -20>, angularTarget=<0 0 0>}
      localJointOffset=<0.4 0 0>
     }]
    relativeSpaceBone {boneName=head}
   }
   n3 = Anim Dynamics - Bone: dyn_hairBang_maskOn_r(ComponentPose=@n2, Alpha=1, ExternalForce="0,0,0") {
    angularDampingOverride=0.9999
    angularSpringConstant=1500
    bAngularSpring=true
    bOverrideAngularDamping=true
    gravityScale=0.35
    alphaBoolBlend {blendOption=Linear}
    boundBone {boneName=dyn_hairBang_maskOn_r}
    componentAppliedLinearAccClamp=<150 150 150>
    componentLinearAccScale=<0.3 0.3 0.3>
    componentLinearVelScale=<0.3 0.3 0.3>
    gravityOverride {z=-700}
    physicsBodyDefinitions [{
      sphereCollisionRadius=0
      boundBone {boneName=dyn_hairBang_maskOn_r}
      boxExtents=<7 6 7>
      constraintSetup {angularLimitsMax=<0 0 30>, angularLimitsMin=<0 -15 -10>, angularTarget=<0 0 0>}
      localJointOffset=<-0.5 0 0>
     }]
   }
   n4 = "Constraint - dyn_hair_00_trgt_l (PPP)"!(ComponentPose=@n3, Alpha=1, ConstraintWeights_0=0.247, ConstraintWeights_1:float=0.753, ConstraintWeights_2:float=0) -> (Pose) {
    alphaBoolBlend {blendOption=Linear}
    boneToModify {boneName=dyn_hair_00_trgt_l}
    constraintSetup [{targetBone {boneName=spine_05}}, {targetBone {boneName=clavicle_l}}, {targetBone {boneName=neck_01}}]
    constraintWeights [1,1,1]
   }
   n5 = "Transform (Modify) Bone - Bone: dyn_hair_00_trgt_l"(ComponentPose=@n4, Alpha=1, Translation="2.082,-0.633,-6.328", Rotation="0,0,0", Scale="1,1,1") {
    translationMode=BMM_Additive
    translationSpace=BCS_BoneSpace
    alphaBoolBlend {blendOption=Linear}
    boneToModify {boneName=dyn_hair_00_trgt_l}
    translation=<2.082 -0.633 -6.328>
   }
   n6 = Get Head look up alpha() -> (Head look up alpha:double)
   n7 = "Transform (Modify) Bone - Bone: dyn_hair_00_trgt_l"(ComponentPose=@n5, Alpha=@n6, Translation="-0.044,4.325,2.764", Rotation="0,0,0", Scale="1,1,1") {
    translationMode=BMM_Additive
    translationSpace=BCS_BoneSpace
    alphaBoolBlend {blendOption=Linear}
    boneToModify {boneName=dyn_hair_00_trgt_l}
    translation=<-0.044 4.325 2.764>
   }
   n8 = "Transform (Modify) Bone - Bone: dyn_hair_base_l"(ComponentPose=@n7, Alpha=@n6, Translation="0,0.225,-0.057", Rotation="-11.963,-9.631,14.007", Scale="1,1,1") {
    rotationMode=BMM_Additive
    rotationSpace=BCS_BoneSpace
    translationMode=BMM_Additive
    translationSpace=BCS_BoneSpace
    alphaBoolBlend {blendInTime=0.3, blendOption=Linear, blendOutTime=0.3}
    boneToModify {boneName=dyn_hair_base_l}
    rotation=<rot -11.963 -9.631 14.007>
    translation {y=0.225, z=-0.057}
   }
   n9 = Look At - Bone: dyn_hair_00_l(ComponentPose=@n8, Alpha=1) {
    alphaBoolBlend {blendOption=Linear}
    boneToModify {boneName=dyn_hair_00_l}
    lookAtLocation {x=0}
    lookAtTarget {boneReference {boneName=dyn_hair_00_trgt_l}}
    lookAt_Axis {axis=<1 0>}
   }
   n10 = Anim Dynamics - Bone: dyn_hair_00_l(ComponentPose=@n9, Alpha=1, ExternalForce="0,0,0") {
    angularDampingOverride=1
    angularSpringConstant=150
    bAngularSpring=true
    bOverrideAngularDamping=true
    bUseSphericalLimits=true
    gravityScale=0.35
    simulationSpace=RootRelative
    alphaBoolBlend {blendOption=Linear}
    boundBone {boneName=dyn_hair_00_l}
    componentAppliedLinearAccClamp=<150 150 200>
    componentLinearAccScale=<0.1 0.1 0.1>
    componentLinearVelScale=<0.3 0.3 0.3>
    physicsBodyDefinitions [{
      collisionType=CustomSphere
      sphereCollisionRadius=2
      boundBone {boneName=dyn_hair_00_l}
      boxExtents=<15 15 15>
      constraintSetup {angularLimitsMax=<10 15 10>, angularLimitsMin=<-10 -20 -10>, angularTarget=<0 0 0>}
      localJointOffset=<4 0 0>
     }]
    planarLimits [{drivingBone {boneName=head}, planeTransform {location=<-0.0545 -5.8503 -3.3295>, rotation=<rot 4.1976 1.7729 -169.2649>}}]
    sphericalLimits [{limitRadius=5.5, drivingBone {boneName=head}, sphereLocalOffset=<0 0 0>}]
   }
   n11 = Anim Dynamics - Bone: dyn_hair_01_00_l(ComponentPose=@n10, Alpha=1, ExternalForce="0,0,0") {
    angularDampingOverride=1
    angularSpringConstant=150
    bAngularSpring=true
    bOverrideAngularDamping=true
    bUseSphericalLimits=true
    gravityScale=0.35
    simulationSpace=RootRelative
    alphaBoolBlend {blendOption=Linear}
    boundBone {boneName=dyn_hair_01_00_l}
    componentAppliedLinearAccClamp=<150 150 200>
    componentLinearAccScale=<0.1 0.1 0.1>
    componentLinearVelScale=<0.3 0.3 0.3>
    physicsBodyDefinitions [{
      collisionType=CustomSphere
      sphereCollisionRadius=2
      boundBone {boneName=dyn_hair_01_00_l}
      boxExtents=<24 24 22>
      constraintSetup {angularLimitsMax=<10 10 5>, angularLimitsMin=<-10 -20 -5>, angularTarget=<0 0 0>}
      localJointOffset=<3 0 0>
     }]
    planarLimits [{drivingBone {boneName=neck_01}, planeTransform {location=<0.6278 -3.3144 -5.1813>, rotation=<rot 4.1426 51.0187 -173.3935>}}]
    sphericalLimits [{limitRadius=8, drivingBone {boneName=clavicle_l}, sphereLocalOffset=<1.199 -3.8017 -3.5879>}]
   }
   n12 = Anim Dynamics - Bone: dyn_hair_02_l(ComponentPose=@n11, Alpha=1, ExternalForce="0,0,0") {
    angularDampingOverride=1
    angularSpringConstant=150
    bAngularSpring=true
    bOverrideAngularDamping=true
    bUseSphericalLimits=true
    gravityScale=0.35
    simulationSpace=RootRelative
    alphaBoolBlend {blendOption=Linear}
    boundBone {boneName=dyn_hair_02_l}
    componentAppliedLinearAccClamp=<150 150 200>
    componentLinearAccScale=<0.1 0.1 0.1>
    componentLinearVelScale=<0.3 0.3 0.3>
    physicsBodyDefinitions [{
      collisionType=CustomSphere
      sphereCollisionRadius=2
      boundBone {boneName=dyn_hair_02_l}
      boxExtents=<33 33 27>
      constraintSetup {angularLimitsMax=<10 5 0>, angularLimitsMin=<-10 -10 0>, angularTarget=<0 0 0>}
      localJointOffset=<4 0 0>
     }]
    planarLimits [
     {drivingBone {boneName=head}, planeTransform {location=<0.2606 -5.181 -5.748>, rotation=<rot -0.0459 11.8762 175.9675>}}
     {drivingBone {boneName=clavicle_l}, planeTransform {location=<-1.0851 -1.0051 3.7535>, rotation=<rot 12.9773 26.3479 46.0576>}}
    ]
   }
   n13 = Anim Dynamics - Bone: dyn_hairStrand_00_l(ComponentPose=@n12, Alpha=1, ExternalForce="0,0,0") {
    angularDampingOverride=0.9999
    angularSpringConstant=1500
    bAngularSpring=true
    bOverrideAngularDamping=true
    bUseSphericalLimits=true
    gravityScale=0.35
    simulationSpace=RootRelative
    alphaBoolBlend {blendOption=Linear}
    boundBone {boneName=dyn_hairStrand_00_l}
    componentAppliedLinearAccClamp=<150 150 200>
    componentLinearAccScale=<0.1 0.1 0.1>
    componentLinearVelScale=<0.3 0.3 0.3>
    gravityOverride {z=1000}
    physicsBodyDefinitions [{
      collisionType=CustomSphere
      sphereCollisionRadius=2
      boundBone {boneName=dyn_hairStrand_00_l}
      boxExtents=<15 15 10>
      constraintSetup {angularLimitsMax=<0 20 20>, angularLimitsMin=<0 -5 -5>, angularTarget=<0 0 0>}
      localJointOffset=<3 0 0>
     }]
    sphericalLimits [{limitRadius=5.5, drivingBone {boneName=head}, sphereLocalOffset=<0 0 0>}]
   }
   n14 = Anim Dynamics - Bone: dyn_hairStrand_01_l(ComponentPose=@n13, Alpha=1, ExternalForce="0,0,0") {
    angularDampingOverride=0.9999
    angularSpringConstant=1500
    bAngularSpring=true
    bOverrideAngularDamping=true
    bUseSphericalLimits=true
    gravityScale=0.35
    simulationSpace=RootRelative
    alphaBoolBlend {blendOption=Linear}
    boundBone {boneName=dyn_hairStrand_01_l}
    componentAppliedLinearAccClamp=<150 150 200>
    componentLinearAccScale=<0.1 0.1 0.1>
    componentLinearVelScale=<0.3 0.3 0.3>
    gravityOverride {z=1000}
    physicsBodyDefinitions [{
      collisionType=CustomSphere
      sphereCollisionRadius=2
      boundBone {boneName=dyn_hairStrand_01_l}
      boxExtents=<15 15 10>
      constraintSetup {angularLimitsMax=<0 20 30>, angularLimitsMin=<0 -15 -18>, angularTarget=<0 0 0>}
      localJointOffset=<3 0 0>
     }]
    planarLimits [{drivingBone {boneName=head}, planeTransform {location=<-0.0189 -11.5605 -4.0952>, rotation=<rot -15.631 -0.9204 -177.0611>}}]
   }
   n15 = "Constraint - dyn_hair_trgt_r (PPP)"!(ComponentPose=@n14, Alpha=1, ConstraintWeights_0=0.332, ConstraintWeights_1:float=0.078, ConstraintWeights_2:float=0.59) -> (Pose) {
    alphaBoolBlend {blendOption=Linear}
    boneToModify {boneName=dyn_hair_trgt_r}
    constraintSetup [{targetBone {boneName=clavicle_r}}, {targetBone {boneName=neck_01}}, {targetBone {boneName=spine_05}}]
    constraintWeights [1,1,1]
   }
   n16 = "Transform (Modify) Bone - Bone: dyn_hair_trgt_r"(ComponentPose=@n15, Alpha=1, Translation="-4.530,4.556,-0.727", Rotation="0,0,0", Scale="1,1,1") {
    translationMode=BMM_Additive
    translationSpace=BCS_BoneSpace
    alphaBoolBlend {blendOption=Linear}
    boneToModify {boneName=dyn_hair_trgt_r}
    translation=<-4.53 4.556 -0.727>
   }
   n17 = Get Head look up alpha()
   n18 = "Transform (Modify) Bone - Bone: dyn_hair_trgt_r"(ComponentPose=@n16, Alpha=@n17, Translation="-0.705,-5.860,-2.441", Rotation="0,0,0", Scale="1,1,1") {
    translationMode=BMM_Additive
    translationSpace=BCS_BoneSpace
    alphaBoolBlend {blendOption=Linear}
    boneToModify {boneName=dyn_hair_trgt_r}
    translation=<-0.705 -5.86 -2.441>
   }
   n19 = "Transform (Modify) Bone - Bone: dyn_hair_base_r"(ComponentPose=@n18, Alpha=@n17, Translation="0.648,-0.285,1.475", Rotation="-43.802,-3.351,-23.120", Scale="1,1,1") {
    rotationMode=BMM_Additive
    rotationSpace=BCS_BoneSpace
    translationMode=BMM_Additive
    translationSpace=BCS_BoneSpace
    alphaBoolBlend {blendInTime=0.3, blendOption=Linear, blendOutTime=0.3}
    boneToModify {boneName=dyn_hair_base_r}
    rotation=<rot -43.802 -3.3509 -23.12>
    translation=<0.648 -0.285 1.475>
   }
   n20 = Look At - Bone: dyn_hair_00_r(ComponentPose=@n19, Alpha=1) {
    alphaBoolBlend {blendOption=Linear}
    boneToModify {boneName=dyn_hair_00_r}
    lookAtLocation {x=0}
    lookAtTarget {boneReference {boneName=dyn_hair_trgt_r}}
    lookAt_Axis {axis=<-1 0>}
    lookUp_Axis {axis {x=-1}}
   }
   n21 = "Transform (Modify) Bone - Bone: dyn_hair_base_r"(
    ComponentPose=@n20
    Alpha=Get Head look down alpha() -> (Head look down alpha:double)
    Translation="0.483,0.715,0.246"
    Rotation="5.063,-2.735,-2.863"
    Scale="1,1,1"
   ) {
    rotationMode=BMM_Additive
    rotationSpace=BCS_BoneSpace
    translationMode=BMM_Additive
    translationSpace=BCS_BoneSpace
    alphaBoolBlend {blendInTime=0.3, blendOption=Linear, blendOutTime=0.3}
    boneToModify {boneName=dyn_hair_base_r}
    rotation=<rot 5.063 -2.735 -2.863>
    translation=<0.483 0.715 0.246>
   }
   n22 = Anim Dynamics - Bone: dyn_hair_00_r(ComponentPose=@n21, Alpha=1, ExternalForce="0,0,0") {
    angularDampingOverride=0.9999
    angularSpringConstant=1500
    bAngularSpring=true
    bOverrideAngularDamping=true
    bUseSphericalLimits=true
    gravityScale=0.35
    simulationSpace=RootRelative
    alphaBoolBlend {blendOption=Linear}
    boundBone {boneName=dyn_hair_00_r}
    componentAppliedLinearAccClamp=<150 150 200>
    componentLinearAccScale=<0.1 0.1 0.1>
    componentLinearVelScale=<0.3 0.3 0.3>
    physicsBodyDefinitions [{
      collisionType=CustomSphere
      sphereCollisionRadius=2
      boundBone {boneName=dyn_hair_00_r}
      boxExtents=<15 15 15>
      constraintSetup {angularLimitsMax=<10 10 8>, angularLimitsMin=<-10 -8 -8>, angularTarget=<0 0 0>}
      localJointOffset=<-3 -1.2864 0.0576>
     }]
    planarLimits [{drivingBone {boneName=head}, planeTransform {location=<0.1494 -0.2748 4.4255>, rotation=<rot 10.8778 7.8434 -101.7766>}}]
   }
   n23 = Anim Dynamics - Bone: dyn_hair_01_00_r(ComponentPose=@n22, Alpha=1, ExternalForce="0,0,0") {
    angularDampingOverride=1
    angularSpringConstant=150
    bAngularSpring=true
    bOverrideAngularDamping=true
    bUseSphericalLimits=true
    gravityScale=0.35
    simulationSpace=RootRelative
    alphaBoolBlend {blendOption=Linear}
    boundBone {boneName=dyn_hair_01_00_r}
    componentAppliedLinearAccClamp=<150 150 200>
    componentLinearAccScale=<0.1 0.1 0.1>
    componentLinearVelScale=<0.3 0.3 0.3>
    physicsBodyDefinitions [{
      collisionType=CustomSphere
      sphereCollisionRadius=2
      boundBone {boneName=dyn_hair_01_00_r}
      boxExtents=<24 24 15>
      constraintSetup {angularLimitsMax=<10 5 5>, angularLimitsMin=<-10 -10 -10>, angularTarget=<0 0 0>}
      localJointOffset=<-3 0 0>
     }]
    planarLimits [{drivingBone {boneName=head}, planeTransform {location=<0.997 -5.8939 4.6743>, rotation=<rot 3.2246 3.864 -5.2749>}}]
    sphericalLimits [{limitRadius=8, drivingBone {boneName=clavicle_r}, sphereLocalOffset=<-4.6031 3.6544 3.3641>}]
   }
   n24 = Anim Dynamics - Bone: dyn_hair_02_r(ComponentPose=@n23, Alpha=1, ExternalForce="0,0,0") {
    angularDampingOverride=1
    angularSpringConstant=150
    bAngularSpring=true
    bOverrideAngularDamping=true
    bUseSphericalLimits=true
    gravityScale=0.35
    simulationSpace=RootRelative
    alphaBoolBlend {blendOption=Linear}
    boundBone {boneName=dyn_hair_02_r}
    componentAppliedLinearAccClamp=<150 150 200>
    componentLinearAccScale=<0.1 0.1 0.1>
    componentLinearVelScale=<0.3 0.3 0.3>
    physicsBodyDefinitions [{
      collisionType=CustomSphere
      sphereCollisionRadius=2
      boundBone {boneName=dyn_hair_02_r}
      boxExtents=<33 33 27>
      constraintSetup {angularLimitsMax=<10 0 0>, angularLimitsMin=<-10 -8 -5>, angularTarget=<0 0 0>}
      localJointOffset=<-1 0 0>
     }]
    planarLimits [
     {drivingBone {boneName=head}, planeTransform {location=<-0.6561 -3.2888 7.2604>, rotation=<rot 7.2485 5.6652 21.3456>}}
     {drivingBone {boneName=clavicle_r}, planeTransform {location=<0.2202 0.4039 -6.5074>, rotation=<rot 20.3148 3.5103 -145.8626>}}
    ]
   }
   n25 = Anim Dynamics - Bone: dyn_hairStrand_00_r(ComponentPose=@n24, Alpha=1, ExternalForce="0,0,0") {
    angularDampingOverride=0.9999
    angularSpringConstant=1500
    bAngularSpring=true
    bOverrideAngularDamping=true
    bUseSphericalLimits=true
    gravityScale=0.35
    simulationSpace=RootRelative
    alphaBoolBlend {blendOption=Linear}
    boundBone {boneName=dyn_hairStrand_00_r}
    componentAppliedLinearAccClamp=<150 150 200>
    componentLinearAccScale=<0.1 0.1 0.1>
    componentLinearVelScale=<0.3 0.3 0.3>
    gravityOverride {z=1000}
    physicsBodyDefinitions [{
      collisionType=CustomSphere
      sphereCollisionRadius=2
      boundBone {boneName=dyn_hairStrand_00_r}
      boxExtents=<10 15 15>
      constraintSetup {angularLimitsMax=<0 20 30>, angularLimitsMin=<0 -5 -12>, angularTarget=<0 0 0>}
      localJointOffset=<-3 0 0>
     }]
    sphericalLimits [{limitRadius=5.4, drivingBone {boneName=head}, sphereLocalOffset=<0 0 0>}]
   }
   n26 = Anim Dynamics - Bone: dyn_hairStrand_01_r(ComponentPose=@n25, Alpha=1, ExternalForce="0,0,0") {
    angularDampingOverride=0.9999
    angularSpringConstant=1500
    bAngularSpring=true
    bOverrideAngularDamping=true
    bUseSphericalLimits=true
    gravityScale=0.35
    simulationSpace=RootRelative
    alphaBoolBlend {blendOption=Linear}
    boundBone {boneName=dyn_hairStrand_01_r}
    componentAppliedLinearAccClamp=<150 150 200>
    componentLinearAccScale=<0.1 0.1 0.1>
    componentLinearVelScale=<0.3 0.3 0.3>
    gravityOverride {z=1000}
    physicsBodyDefinitions [{
      collisionType=CustomSphere
      sphereCollisionRadius=2
      boundBone {boneName=dyn_hairStrand_01_r}
      boxExtents=<10 15 15>
      constraintSetup {angularLimitsMax=<0 20 25>, angularLimitsMin=<0 -15 -12>, angularTarget=<0 0 0>}
      localJointOffset=<-3 0 0>
     }]
    planarLimits [{drivingBone {boneName=head}, planeTransform {location=<0.3829 -9.9065 5.5604>, rotation=<rot 13.405 1.1204 5.6386>}}]
   }
   n27 = "Constraint - dyn_hairBack_trgt (PPPP)"!(
    ComponentPose=@n26
    Alpha=0.7
    ConstraintWeights_0=0.057
    ConstraintWeights_1:float=0.057
    ConstraintWeights_2:float=0.029
    ConstraintWeights_3:float=0.857
   ) -> (Pose) {
    alphaBoolBlend {blendOption=Linear}
    boneToModify {boneName=dyn_hairBack_trgt}
    constraintSetup [{targetBone {boneName=clavicle_l}}, {targetBone {boneName=clavicle_r}}, {targetBone {boneName=neck_01}}, {targetBone {boneName=spine_05}}]
    constraintWeights [1,1,1,1]
   }
   n28 = "Transform (Modify) Bone - Bone: dyn_hairBack_trgt"(ComponentPose=@n27, Alpha=Get Head look down alpha(), Translation="0,0,-1.953", Rotation="0,0,0", Scale="1,1,1") {
    translationMode=BMM_Additive
    translationSpace=BCS_BoneSpace
    alphaBoolBlend {blendOption=Linear}
    boneToModify {boneName=dyn_hairBack_trgt}
    translation {z=-1.953}
   }
   n29 = Look At - Bone: dyn_hairBack_00(ComponentPose=@n28, Alpha=1) {
    alphaBoolBlend {blendOption=Linear}
    boneToModify {boneName=dyn_hairBack_00}
    lookAtLocation {x=0}
    lookAtTarget {boneReference {boneName=dyn_hairBack_trgt}}
    lookAt_Axis {axis=<1 0>}
   }
   n30 = Anim Dynamics - Bone: dyn_hairBack_00(ComponentPose=@n29, Alpha=1, ExternalForce="0,0,0") {
    angularDampingOverride=0.9999
    angularSpringConstant=1500
    bAngularSpring=true
    bOverrideAngularDamping=true
    simulationSpace=RootRelative
    alphaBoolBlend {blendOption=Linear}
    boundBone {boneName=dyn_hairBack_00}
    componentAppliedLinearAccClamp=<150 150 200>
    componentLinearAccScale=<0.1 0.1 0.1>
    componentLinearVelScale=<0.3 0.3 0.3>
    physicsBodyDefinitions [{
      boundBone {boneName=dyn_hairBack_00}
      boxExtents=<25 25 20>
      constraintSetup {angularLimitsMax=<5 15 10>, angularLimitsMin=<-5 -15 -10>, angularTarget=<0 0 0>}
      localJointOffset=<5 0 0>
     }]
   }
   n31 = Anim Dynamics - Bone: dyn_hairBack_01(ComponentPose=@n30, Alpha=1, ExternalForce="0,0,0") {
    angularDampingOverride=0.9999
    angularSpringConstant=1500
    bAngularSpring=true
    bOverrideAngularDamping=true
    bUseSphericalLimits=true
    simulationSpace=RootRelative
    alphaBoolBlend {blendOption=Linear}
    boundBone {boneName=dyn_hairBack_01}
    componentAppliedLinearAccClamp=<150 150 200>
    componentLinearAccScale=<0.1 0.1 0.1>
    componentLinearVelScale=<0.3 0.3 0.3>
    physicsBodyDefinitions [{
      collisionType=CustomSphere
      sphereCollisionRadius=2
      boundBone {boneName=dyn_hairBack_01}
      boxExtents=<30 30 25>
      constraintSetup {angularLimitsMax=<5 15 0>, angularLimitsMin=<-5 -25 0>, angularTarget=<0 0 0>}
      localJointOffset=<5 0 0>
     }]
    planarLimits [{drivingBone {boneName=spine_05}, planeTransform {location=<0.1194 -6.0786 0.1084>, rotation=<rot 0.6803 22.1431 -90.3633>}}]
   }
   n32 = "Constraint - dyn_hairBack_cstr (RRR)"!(ComponentPose=@n31, Alpha=1, ConstraintWeights_0=0.4, ConstraintWeights_1:float=0.4, ConstraintWeights_2:float=0.2) -> (Pose) {
    alphaBoolBlend {blendOption=Linear}
    boneToModify {boneName=dyn_hairBack_cstr}
    constraintSetup [
     {transformType=Rotation, targetBone {boneName=dyn_hair_00_l}}
     {transformType=Rotation, targetBone {boneName=dyn_hair_00_r}}
     {transformType=Rotation, targetBone {boneName=spine_05}}
    ]
    constraintWeights [1,1,1]
   }
   n33 = Get bIsJumping()
   n34 = "Transform (Modify) Bone - Bone: dyn_hair_base_l"!(ComponentPose=@n32, bAlphaBoolEnabled:bool=@n33, Translation="0,0,0", Rotation="10,0,0", Scale="1,1,1") -> (Pose) {
    alphaInputType=Bool
    rotationMode=BMM_Additive
    rotationSpace=BCS_BoneSpace
    translationMode=BMM_Additive
    alphaBoolBlend {blendInTime=0.3, blendOption=Linear, blendOutTime=0.5}
    boneToModify {boneName=dyn_hair_base_l}
    rotation {pitch=10}
   }
   n35 = "Transform (Modify) Bone - Bone: dyn_hairBack_base"!(ComponentPose=@n34, bAlphaBoolEnabled:bool=@n33, Translation="0,0,0", Rotation="15,0,0", Scale="1,1,1") -> (Pose) {
    alphaInputType=Bool
    rotationMode=BMM_Additive
    rotationSpace=BCS_BoneSpace
    translationMode=BMM_Additive
    alphaBoolBlend {blendInTime=0.2, blendOption=Linear, blendOutTime=0.4}
    boneToModify {boneName=dyn_hairBack_base}
    rotation {pitch=15}
   }
   n36 = "Transform (Modify) Bone - Bone: dyn_hair_base_r"!(ComponentPose=@n35, bAlphaBoolEnabled:bool=@n33, Translation="0,0,0", Rotation="12,0,0", Scale="1,1,1") -> (Pose) {
    alphaInputType=Bool
    rotationMode=BMM_Additive
    rotationSpace=BCS_BoneSpace
    translationMode=BMM_Additive
    alphaBoolBlend {blendInTime=0.2, blendOption=Linear, blendOutTime=0.3}
    boneToModify {boneName=dyn_hair_base_r}
    rotation {pitch=12}
   }
   Output Pose(Result=Component To Local(ComponentPose=@n36)) {layerGroup=None}
  ]}
 EventGraph {nodes [
   n1 = Event BlueprintUpdateAnimation()
   n2 = Set Skeletal mesh(execute=@n7.then, Skeletal mesh:Skeletal Mesh Component Object Reference=GetPartSkeletalMeshComponent()) -> (Output_Get:Skeletal Mesh Component Object Reference)
   n3 = NOT Boolean(A=Get bIsPlayingMeleeAnim())
   Set Head look up alpha(
    execute=@n1.then
    Head look up alpha:double=SelectFloat(
     A=GetAlphaFromReferenceTransformDelta(
      SkeletalMeshComponent=@n2
      BoneName=head
      BaseName=spine_05
      DeltaComponentType=EulerY
      InputRange=(X=5,Y=15)
      OutputRange=(X=0,Y=0.6)
      TwistAxis="1,0,0"
     )
     B=0
     bPickA=@n3
    )
   ) -> (Output_Get:double)
   n4 = GetAlphaFromReferenceTransformDelta(
    SkeletalMeshComponent=@n2
    BoneName=head
    BaseName=spine_05
    DeltaComponentType=EulerY
    InputRange=(X=-15,Y=5)
    OutputRange=(X=1,Y=0)
    TwistAxis="1,0,0"
   )
   > Set Head look down alpha(Head look down alpha:double=SelectFloat(A=@n4, B=0, bPickA=@n3)) -> (Output_Get:double)
   > Set Machine gun corr(Machine gun corr:double=SelectFloat(A=@n4, B=0, bPickA=Get Machine gun equiped() -> (Machine gun equiped:bool))) -> (Output_Get:double)
   > n5 = Cast To FortPlayerAnimInstance(Object=GetAnimInstance(self=GetOwnerSkeletalMeshComponent()))
   > Set Machine gun equiped(Machine gun equiped:bool=false) -> (Output_Get:bool)
   > n6 = Switch on EFortWeaponCoreAnimation!(execute, Selection:EFortWeaponCoreAnimation Enum=Get WeaponCoreAnim(self=@n5)) -> (
    Melee:Exec
    Pistol:Exec
    Shotgun:Exec
    PaperBlueprint:Exec
    Rifle:Exec
    MeleeOneHand:Exec
    MachinePistol:Exec
    RocketLauncher:Exec
    GrenadeLauncher:Exec
    GoingCommando:Exec
    AssaultRifle:Exec
    TacticalShotgun:Exec
    SniperRifle:Exec
    TrapPlacement:Exec
    ShoulderLauncher:Exec
    AbilityDecoTool:Exec
    Crossbow:Exec
    C4:Exec
    RemoteControl:Exec
    DualWield:Exec
    AR_BullPup:Exec
    AR_ForwardGrip:Exec
    MedPackPaddles:Exec
    SMG_P90:Exec
    AR_DrumGun:Exec
    Consumable_Small:Exec
    Consumable_Large:Exec
    Balloon:Exec
    MountedTurret:Exec
    CreativeTool:Exec
    ExplosiveBow:Exec
    AshtonIndigo:Exec
    AshtonChicago:Exec
    MeleeDualWield:Exec
    Unarmed:Exec
   )
   Branch(execute=@n6.GoingCommando)
   > Set Machine gun equiped(Machine gun equiped=true)
   n7 = Event OnCharacterCustomizationLoaded()
  ]}
]
variables [
 Head look down alpha {type=number}
 Head look up alpha {type=number}
 Machine gun corr {type=number}
 Machine gun equiped {type=boolean}
 Skeletal mesh {type=SkeletalMeshComponent}
]
